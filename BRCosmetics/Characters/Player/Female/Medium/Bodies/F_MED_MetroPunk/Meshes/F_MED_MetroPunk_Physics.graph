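# Physics asset for the MetroPunk medium female player body, set up for dynamic motion of the jacket.
# Front, side and back jacket chains on each side run from zero-size base bodies through two capsules with limited swing and twist.
# Cross constraints tie neighboring lower panels together, joining left and right only at the front.
# Pelvis, lower spine, thigh and foot capsules supply body shapes, linked by free-motion constraints, as are the jacket bases to the pelvis.
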PhysicsAsset /BRCosmetics/Characters/Player/Female/Medium/Bodies/F_MED_MetroPunk/Meshes/F_MED_MetroPunk_Physics
bodies {
 dyn_jacket_back_01_l [{
   shapeType=Capsule
   length=13.4
   radius=6.3
   shapeName=dyn_jacket_side_02_l_capsule
   center=<-6.3074 -0.0576 -2.7551>
   rotation=<rot 0.0001 179.9998 86.6815>
  }]
 dyn_jacket_back_01_r [{shapeType=Capsule, length=13.4, radius=6.3, shapeName=dyn_jacket_back_01_r_capsule, center=<6.5264 0 3.0691>, rotation=<rot 0 0 -90>}]
 dyn_jacket_back_02_l [{
   shapeType=Capsule
   length=24.4
   radius=6.9
   shapeName=dyn_jacket_side_02_l_capsule
   center=<-6.3074 0.0685 -0.5794>
   rotation=<rot 0.0399 -179.3148 86.6817>
  }]
 dyn_jacket_back_02_r [{shapeType=Capsule, length=24.4, radius=6.9, shapeName=dyn_jacket_back_02_r_capsule, center=<6.5264 -0.583 0.8897>, rotation=<rot 0 0 -90>}]
 dyn_jacket_back_base_l [{shapeType=Capsule, length=0, radius=0, shapeName=dyn_jacket_back_base_l_capsule, rotation=<rot 0 -90 0>}]
 dyn_jacket_back_base_r [{shapeType=Capsule, length=0, radius=0, shapeName=dyn_jacket_back_base_r_capsule, rotation=<rot 0 -90 0>}]
 dyn_jacket_front_01_l [{
   shapeType=Capsule
   length=25
   radius=6.9
   shapeName=dyn_jacket_front_01_l_capsule
   center=<-5 -0.1 -2.7>
   rotation=<rot 0.068 179.9779 86.0006>
  }]
 dyn_jacket_front_01_r [{shapeType=Capsule, length=25, radius=6.9, shapeName=dyn_jacket_front_01_r_capsule, center=<5 0.3 3.2>, rotation=<rot 0 0 -80>}]
 dyn_jacket_front_02_l [{shapeType=Capsule, length=33.9, radius=13, shapeName=dyn_jacket_front_02_l_capsule, center=<-5 -3.6725 -6.4497>, rotation=<rot 0 180 86>}]
 dyn_jacket_front_02_r [{
   shapeType=Capsule
   length=46.7001
   radius=6.9
   shapeName=dyn_jacket_front_02_r_capsule
   center=<5 -1.5 0.7>
   rotation=<rot 0.0003 -0.0001 -73.5125>
  }]
 dyn_jacket_front_base_l [{shapeType=Capsule, length=0, radius=0, shapeName=dyn_jacket_front_base_l_capsule, rotation=<rot 0 -90 0>}]
 dyn_jacket_front_base_r [{shapeType=Capsule, length=0, radius=0, shapeName=dyn_jacket_front_base_r_capsule, rotation=<rot 0 -90 0>}]
 dyn_jacket_side_01_l [{
   shapeType=Capsule
   length=17.9
   radius=5.7
   shapeName=dyn_jacket_side_02_l_capsule
   center=<-6.2815 0 -2.5743>
   rotation=<rot 0 179.9999 88.0329>
  }]
 dyn_jacket_side_01_r [{shapeType=Capsule, length=17.9, radius=5.7, shapeName=dyn_jacket_side_01_r_capsule, center=<5.9619 0 2.6086>, rotation=<rot 0 0 -90>}]
 dyn_jacket_side_02_l [{shapeType=Capsule, length=24, radius=6.9, shapeName=dyn_jacket_side_02_l_capsule, center=<-6.2815 0 -3>, rotation=<rot 0 179.9999 88.0329>}]
 dyn_jacket_side_02_r [{shapeType=Capsule, length=23, radius=6.9, shapeName=dyn_jacket_side_02_r_capsule, center=<5.9619 0 4.0173>, rotation=<rot 0 0 -90>}]
 dyn_jacket_side_base_l [{shapeType=Capsule, length=0, radius=0, shapeName=dyn_jacket_side_base_l_capsule, rotation=<rot 0 -90 0>}]
 dyn_jacket_side_base_r [{shapeType=Capsule, length=0, radius=0, shapeName=dyn_jacket_side_base_r_capsule, rotation=<rot 0 -90 0>}]
 foot_l [{shapeType=Capsule, length=22.6124, radius=7.957, shapeName=foot_l_capsule, center=<-4.1512 6.9324 0.0146>, rotation=<rot 0 -0.0001 -88.6526>}]
 foot_r [{shapeType=Capsule, length=22.6124, radius=7.957, shapeName=foot_r_capsule, center=<4.1512 -6.9324 -0.0146>, rotation=<rot 0 -0.0001 91.3474>}]
 pelvis [{shapeType=Capsule, length=7.0824, radius=12.2, shapeName=pelvis_capsule, center=<-3.8018 -0.0254 -0.0104>, rotation=<rot 89.7643 -0.0002 0>}]
 spine_01 [{shapeType=Capsule, length=7.3, radius=10, shapeName=spine_01_capsule}]
 thigh_l [{shapeType=Capsule, length=36, radius=6.4, shapeName=thigh_l_capsule, center=<-26.198 0 0.2146>, rotation=<rot 90 0 180>}]
 thigh_r [{shapeType=Capsule, length=32.4, radius=6.4, shapeName=thigh_r_capsule, center=<24.6438 0 -0.2146>, rotation=<rot 90 0 0>}]
}
constraints [
 {bone1Name=dyn_jacket_back_01_l, bone2Name=dyn_jacket_back_base_l, swing1LimitDegrees=15, swing2LimitDegrees=90, twistLimitDegrees=5}
 {bone1Name=dyn_jacket_back_02_l, bone2Name=dyn_jacket_back_01_l, swing1LimitDegrees=15, swing2LimitDegrees=45, twistLimitDegrees=5}
 {bone1Name=dyn_jacket_back_01_r, bone2Name=dyn_jacket_back_base_r, swing1LimitDegrees=15, swing2LimitDegrees=90, twistLimitDegrees=5}
 {bone1Name=dyn_jacket_back_02_r, bone2Name=dyn_jacket_back_01_r, swing1LimitDegrees=15, swing2LimitDegrees=45, twistLimitDegrees=5}
 {bone1Name=dyn_jacket_front_01_l, bone2Name=dyn_jacket_front_base_l, swing1LimitDegrees=25, swing2LimitDegrees=65, twistLimitDegrees=6}
 {bone1Name=dyn_jacket_front_02_l, bone2Name=dyn_jacket_front_01_l, swing1LimitDegrees=5, swing2LimitDegrees=25, twistLimitDegrees=5}
 {bone1Name=dyn_jacket_front_01_r, bone2Name=dyn_jacket_front_base_r, swing1LimitDegrees=25, swing2LimitDegrees=65, twistLimitDegrees=6}
 {bone1Name=dyn_jacket_front_02_r, bone2Name=dyn_jacket_front_01_r, swing1LimitDegrees=5, swing2LimitDegrees=25, twistLimitDegrees=5}
 {bone1Name=dyn_jacket_side_01_l, bone2Name=dyn_jacket_side_base_l, swing1LimitDegrees=5, swing2LimitDegrees=90, twistLimitDegrees=5}
 {bone1Name=dyn_jacket_side_02_l, bone2Name=dyn_jacket_side_01_l, swing1LimitDegrees=15, swing2LimitDegrees=180, twistLimitDegrees=5}
 {bone1Name=dyn_jacket_side_01_r, bone2Name=dyn_jacket_side_base_r, swing1LimitDegrees=5, swing2LimitDegrees=90, twistLimitDegrees=35}
 {bone1Name=dyn_jacket_side_02_r, bone2Name=dyn_jacket_side_01_r, swing1LimitDegrees=15, swing2LimitDegrees=45, twistLimitDegrees=5}
 {bone1Name=dyn_jacket_side_02_l, bone2Name=dyn_jacket_front_02_l, swing1LimitDegrees=45, swing2LimitDegrees=45, twistLimitDegrees=45}
 {bone1Name=dyn_jacket_back_02_l, bone2Name=dyn_jacket_side_02_l, swing1LimitDegrees=30, swing2LimitDegrees=45, twistLimitDegrees=45}
 {bone1Name=dyn_jacket_side_02_r, bone2Name=dyn_jacket_front_02_r, swing1LimitDegrees=45, swing2LimitDegrees=45, twistLimitDegrees=45}
 {bone1Name=dyn_jacket_back_02_r, bone2Name=dyn_jacket_side_02_r, swing1LimitDegrees=30, swing2LimitDegrees=45, twistLimitDegrees=45}
 {bone1Name=dyn_jacket_front_02_l, bone2Name=dyn_jacket_front_02_r, swing1LimitDegrees=45, swing2LimitDegrees=45, twistLimitDegrees=45}
 {
  bone1Name=spine_01
  bone2Name=pelvis
  swing1LimitDegrees=45
  swing1Motion=Free
  swing2LimitDegrees=45
  swing2Motion=Free
  twistLimitDegrees=45
  twistMotion=Free
 }
 {
  bone1Name=thigh_l
  bone2Name=pelvis
  swing1LimitDegrees=45
  swing1Motion=Free
  swing2LimitDegrees=45
  swing2Motion=Free
  twistLimitDegrees=45
  twistMotion=Free
 }
 {
  bone1Name=thigh_r
  bone2Name=pelvis
  swing1LimitDegrees=45
  swing1Motion=Free
  swing2LimitDegrees=45
  swing2Motion=Free
  twistLimitDegrees=45
  twistMotion=Free
 }
 {
  bone1Name=dyn_jacket_front_base_r
  bone2Name=pelvis
  swing1LimitDegrees=45
  swing1Motion=Free
  swing2LimitDegrees=45
  swing2Motion=Free
  twistLimitDegrees=45
  twistMotion=Free
 }
 {
  bone1Name=dyn_jacket_side_base_r
  bone2Name=pelvis
  swing1LimitDegrees=45
  swing1Motion=Free
  swing2LimitDegrees=45
  swing2Motion=Free
  twistLimitDegrees=45
  twistMotion=Free
 }
 {
  bone1Name=dyn_jacket_back_base_r
  bone2Name=pelvis
  swing1LimitDegrees=45
  swing1Motion=Free
  swing2LimitDegrees=45
  swing2Motion=Free
  twistLimitDegrees=45
  twistMotion=Free
 }
 {
  bone1Name=dyn_jacket_front_base_l
  bone2Name=pelvis
  swing1LimitDegrees=45
  swing1Motion=Free
  swing2LimitDegrees=45
  swing2Motion=Free
  twistLimitDegrees=45
  twistMotion=Free
 }
 {
  bone1Name=dyn_jacket_side_base_l
  bone2Name=pelvis
  swing1LimitDegrees=45
  swing1Motion=Free
  swing2LimitDegrees=45
  swing2Motion=Free
  twistLimitDegrees=45
  twistMotion=Free
 }
 {
  bone1Name=dyn_jacket_back_base_l
  bone2Name=pelvis
  swing1LimitDegrees=45
  swing1Motion=Free
  swing2LimitDegrees=45
  swing2Motion=Free
  twistLimitDegrees=45
  twistMotion=Free
 }
 {
  bone1Name=thigh_l
  bone2Name=dyn_jacket_side_02_l
  swing1LimitDegrees=45
  swing1Motion=Free
  swing2LimitDegrees=45
  swing2Motion=Free
  twistLimitDegrees=45
  twistMotion=Free
 }
 {
  bone1Name=thigh_r
  bone2Name=dyn_jacket_side_02_r
  swing1LimitDegrees=25
  swing1Motion=Free
  swing2LimitDegrees=25
  swing2Motion=Free
  twistLimitDegrees=25
  twistMotion=Free
 }
 {
  bone1Name=thigh_l
  bone2Name=thigh_r
  swing1LimitDegrees=45
  swing1Motion=Free
  swing2LimitDegrees=45
  swing2Motion=Free
  twistLimitDegrees=45
  twistMotion=Free
 }
 {
  bone1Name=foot_l
  bone2Name=thigh_l
  swing1LimitDegrees=45
  swing1Motion=Free
  swing2LimitDegrees=45
  swing2Motion=Free
  twistLimitDegrees=45
  twistMotion=Free
 }
 {
  bone1Name=foot_r
  bone2Name=thigh_r
  swing1LimitDegrees=45
  swing1Motion=Free
  swing2LimitDegrees=45
  swing2Motion=Free
  twistLimitDegrees=45
  twistMotion=Free
 }
]
