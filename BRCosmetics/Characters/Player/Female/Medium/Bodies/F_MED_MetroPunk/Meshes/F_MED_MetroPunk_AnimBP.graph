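# Post-process style AnimBP for the F_MED_MetroPunk female body, copying the attached parent's pose and adding secondary motion to clothing.
# Weighted constraints drive jacket pocket and zip-base bones, then a Control Rig, AnimDynamics zipper bones and a pelvis-space RigidBody follow.
# EventGraph feeds LOD, rigid-body, melee, grind-rail and short-backbling state to the rig; RigidBody switches off while grinding.
# The deprecated gravity-override result is stored but unused by the AnimGraph.
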AnimBlueprint /BRCosmetics/Characters/Player/Female/Medium/Bodies/F_MED_MetroPunk/Meshes/F_MED_MetroPunk_AnimBP
target_skeleton=&/F_MED_MetroPunk_Skeleton
graphs [
 AnimGraph {nodes [
   n1 = "Constraint - mstr_pocket_jacket_01 (PPPPPPPP)"!(
    ComponentPose=Local To Component(LocalPose=Copy Pose From Mesh() {bUseAttachedParent=true})
    Alpha=1
    ConstraintWeights_0=0.1444
    ConstraintWeights_1:float=0.0146
    ConstraintWeights_2:float=0.0838
    ConstraintWeights_3:float=0
    ConstraintWeights_4:float=0.3019
    ConstraintWeights_5:float=0.3474
    ConstraintWeights_6:float=0
    ConstraintWeights_7:float=0.1078
   ) -> (Pose) {
    alphaBoolBlend {blendOption=Linear}
    boneToModify {boneName=mstr_pocket_jacket_01}
    constraintSetup [
     {targetBone {boneName=upperarm_l}}
     {targetBone {boneName=upperarm_twist_01_l}}
     {targetBone {boneName=spine_04}}
     {targetBone {boneName=spine_03}}
     {targetBone {boneName=spine_05}}
     {targetBone {boneName=clavicle_l}}
     {targetBone {boneName=deform_upperarm_bk_l}}
     {targetBone {boneName=deform_upperarm_up_l}}
    ]
    constraintWeights [1,1,1,1,1,1,1,1]
   }
   n2 = "Constraint - mstr_pocket_jacket_02 (PPPPPPPP)"!(
    ComponentPose=@n1
    Alpha=1
    ConstraintWeights_0=0.2237
    ConstraintWeights_1:float=0.0528
    ConstraintWeights_2:float=0.1884
    ConstraintWeights_3:float=0
    ConstraintWeights_4:float=0.1751
    ConstraintWeights_5:float=0.3067
    ConstraintWeights_6:float=0
    ConstraintWeights_7:float=0.0533
   ) -> (Pose) {
    alphaBoolBlend {blendOption=Linear}
    boneToModify {boneName=mstr_pocket_jacket_02}
    constraintSetup [
     {targetBone {boneName=upperarm_l}}
     {targetBone {boneName=upperarm_twist_01_l}}
     {targetBone {boneName=spine_04}}
     {targetBone {boneName=spine_03}}
     {targetBone {boneName=spine_05}}
     {targetBone {boneName=clavicle_l}}
     {targetBone {boneName=deform_upperarm_bk_l}}
     {targetBone {boneName=deform_upperarm_up_l}}
    ]
    constraintWeights [1,1,1,1,1,1,1,1]
   }
   n3 = "Constraint - dyn_zip_jacket_base (PPPPPPPP)"!(
    ComponentPose=@n2
    Alpha=1
    ConstraintWeights_0=0.0191
    ConstraintWeights_1:float=0.5208
    ConstraintWeights_2:float=0.2812
    ConstraintWeights_3:float=0.176
    ConstraintWeights_4:float=0.0007
    ConstraintWeights_5:float=0.0021
    ConstraintWeights_6:float=0
    ConstraintWeights_7:float=0
   ) -> (Pose) {
    alphaBoolBlend {blendOption=Linear}
    boneToModify {boneName=dyn_zip_jacket_base}
    constraintSetup [
     {targetBone {boneName=spine_02}}
     {targetBone {boneName=spine_04}}
     {targetBone {boneName=spine_03}}
     {targetBone {boneName=spine_05}}
     {targetBone {boneName=clavicle_l}}
     {targetBone {boneName=clavicle_r}}
     {targetBone {boneName=dyn_jacket_front_01_r}}
     {targetBone {boneName=dyn_jacket_side_01_r}}
    ]
    constraintWeights [1,1,1,1,1,1,1,1]
   }
   n4 = Control Rig!(
    Source=Component To Local(ComponentPose=@n3)
    Alpha
    LOD 2:bool=Get LOD2() -> (LOD2:bool)
    RBAN OFF:bool=Get RBAN_Disabled() -> (RBAN_Disabled:bool)
    LOD 1:bool=Get LOD1() -> (LOD1:bool)
    Melee:bool=Get bIsPlayingMeleeAnim()
    On Grind Rail:bool=Get On Grind Rail() -> (On Grind Rail:bool)
    Wear Backpack:bool=Get Wear Backpack() -> (Wear Backpack:bool)
   ) -> (Pose) {
    alphaBoolBlend {blendOption=Linear}
    controlRigAssetReference {blueprintRigClass=&/F_MED_MetroPunk_CtrlRig.F_MED_MetroPunk_CtrlRig_C}
   }
   n5 = Anim Dynamics - Bone: dyn_zip_jacket_01(ComponentPose=Local To Component(LocalPose=@n4), Alpha=1, ExternalForce="0,0,0") {
    angularDampingOverride=0.997
    angularSpringConstant=250
    bAngularSpring=true
    bOverrideAngularDamping=true
    bUseGravityOverride=true
    linearDampingOverride=1
    linearSpringConstant=250
    simulationSpace=RootRelative
    alphaBoolBlend {blendOption=Linear}
    boundBone {boneName=dyn_zip_jacket_01}
    componentAppliedLinearAccClamp=<5000 5000 5000>
    componentLinearAccScale=<0.05 0.05 0.05>
    componentLinearVelScale=<0.05 0.05 0.05>
    gravityOverride {z=-980}
    physicsBodyDefinitions [{
      boundBone {boneName=dyn_zip_jacket_01}
      boxExtents=<10 10 10>
      constraintSetup {angularLimitsMax=<1 -5 20>, angularLimitsMin=<-1 -90 -20>, angularTarget=<1 0 0>}
      localJointOffset=<1 0 0>
     }]
   }
   n6 = Anim Dynamics - Bone: dyn_zip_jacket_02(ComponentPose=@n5, Alpha=1, ExternalForce="0,0,0") {
    angularDampingOverride=0.997
    angularSpringConstant=250
    bAngularSpring=true
    bOverrideAngularDamping=true
    bUseGravityOverride=true
    linearDampingOverride=1
    linearSpringConstant=250
    simulationSpace=RootRelative
    alphaBoolBlend {blendOption=Linear}
    boundBone {boneName=dyn_zip_jacket_02}
    componentAppliedLinearAccClamp=<5000 5000 5000>
    componentLinearAccScale=<0.05 0.05 0.05>
    componentLinearVelScale=<0.05 0.05 0.05>
    gravityOverride {z=-980}
    physicsBodyDefinitions [{
      boundBone {boneName=dyn_zip_jacket_02}
      boxExtents=<10 10 10>
      constraintSetup {angularLimitsMax=<1 0 20>, angularLimitsMin=<-1 -75 -20>, angularTarget=<1 0 0>}
      localJointOffset=<1 0 0>
     }]
   }
   n7 = Anim Dynamics - Bone: dyn_zip_leg_l(ComponentPose=@n6, Alpha=1, ExternalForce="0,0,0") {
    angularDampingOverride=0.997
    angularSpringConstant=250
    bAngularSpring=true
    bOverrideAngularDamping=true
    bUseGravityOverride=true
    linearDampingOverride=1
    linearSpringConstant=250
    simulationSpace=RootRelative
    alphaBoolBlend {blendOption=Linear}
    boundBone {boneName=dyn_zip_leg_l}
    componentAppliedLinearAccClamp=<5000 5000 5000>
    componentLinearAccScale=<0.05 0.05 0.05>
    componentLinearVelScale=<0.05 0.05 0.05>
    gravityOverride {z=-980}
    physicsBodyDefinitions [{
      boundBone {boneName=dyn_zip_leg_l}
      boxExtents=<10 10 10>
      constraintSetup {angularLimitsMax=<1 3 20>, angularLimitsMin=<-1 -90 -20>, angularTarget=<1 0 0>}
      localJointOffset=<1 0 0>
     }]
   }
   n8 = Anim Dynamics - Bone: dyn_zip_leg_r(ComponentPose=@n7, Alpha=1, ExternalForce="0,0,0") {
    angularDampingOverride=0.997
    angularSpringConstant=250
    bAngularSpring=true
    bOverrideAngularDamping=true
    bUseGravityOverride=true
    linearDampingOverride=1
    linearSpringConstant=250
    simulationSpace=RootRelative
    alphaBoolBlend {blendOption=Linear}
    boundBone {boneName=dyn_zip_leg_r}
    componentAppliedLinearAccClamp=<5000 5000 5000>
    componentLinearAccScale=<0.05 0.05 0.05>
    componentLinearVelScale=<0.05 0.05 0.05>
    gravityOverride {z=-980}
    physicsBodyDefinitions [{
      boundBone {boneName=dyn_zip_leg_r}
      boxExtents=<10 10 10>
      constraintSetup {angularLimitsMax=<1 3 20>, angularLimitsMin=<-1 -90 -20>, angularTarget=<1 0 0>}
      localJointOffset=<1 0 0>
     }]
   }
   n9 = Anim Dynamics - Bone: dyn_zip_pocket_leg_01_l(ComponentPose=@n8, Alpha=1, ExternalForce="0,0,0") {
    angularDampingOverride=0.997
    angularSpringConstant=250
    bAngularSpring=true
    bOverrideAngularDamping=true
    bUseGravityOverride=true
    linearDampingOverride=1
    linearSpringConstant=250
    simulationSpace=RootRelative
    alphaBoolBlend {blendOption=Linear}
    boundBone {boneName=dyn_zip_pocket_leg_01_l}
    componentAppliedLinearAccClamp=<5000 5000 5000>
    componentLinearAccScale=<0.05 0.05 0.05>
    componentLinearVelScale=<0.05 0.05 0.05>
    gravityOverride {z=-980}
    physicsBodyDefinitions [{
      boundBone {boneName=dyn_zip_pocket_leg_01_l}
      boxExtents=<10 10 10>
      constraintSetup {angularLimitsMax=<1 2 20>, angularLimitsMin=<-1 -160 -20>, angularTarget=<1 0 0>}
      localJointOffset=<1 0 0>
     }]
   }
   n10 = Anim Dynamics - Bone: dyn_zip_pocket_leg_02_l(ComponentPose=@n9, Alpha=1, ExternalForce="0,0,0") {
    angularDampingOverride=0.997
    angularSpringConstant=250
    bAngularSpring=true
    bOverrideAngularDamping=true
    bUseGravityOverride=true
    linearDampingOverride=1
    linearSpringConstant=250
    simulationSpace=RootRelative
    alphaBoolBlend {blendOption=Linear}
    boundBone {boneName=dyn_zip_pocket_leg_02_l}
    componentAppliedLinearAccClamp=<5000 5000 5000>
    componentLinearAccScale=<0.05 0.05 0.05>
    componentLinearVelScale=<0.05 0.05 0.05>
    gravityOverride {z=-980}
    physicsBodyDefinitions [{
      boundBone {boneName=dyn_zip_pocket_leg_02_l}
      boxExtents=<10 10 10>
      constraintSetup {angularLimitsMax=<1 5 20>, angularLimitsMin=<-1 -160 -20>, angularTarget=<1 0 0>}
      localJointOffset=<1 0 0>
     }]
   }
   n11 = RigidBody!(
    ComponentPose=@n10
    bAlphaBoolEnabled:bool=Get Rban Disabled on Grind Rail() -> (Rban Disabled on Grind Rail:bool)
    OverrideWorldGravity:Vector="0,0,-980"
    ExternalForce
   ) -> (Pose) {
    alphaInputType=Bool
    bClampLinearTranslationLimitToRefPose=true
    bOverrideWorldGravity=true
    overridePhysicsAsset=&/F_MED_MetroPunk_Physics
    simulationSpace=BaseBoneSpace
    alphaBoolBlend {blendOption=Linear}
    baseBoneRef {boneName=pelvis}
    componentAppliedLinearAccClamp=<3500 3500 3500>
    overrideWorldGravity {z=-980}
    simSpaceSettings {maxAngularAcceleration=100, maxAngularVelocity=3, maxLinearAcceleration=100, maxLinearVelocity=100, velocityScaleZ=0, worldAlpha=0.9}
   }
   Output Pose(Result=Component To Local(ComponentPose=@n11)) {layerGroup=None}
  ]}
 EventGraph {nodes [
   Event BlueprintUpdateAnimation()
   > n1 = Sequence!(execute) -> (then_0, then_1, then_2:Exec, then_3:Exec, then_4:Exec, then_5:Exec, then_6:Exec)
   > n2 = Gravity Override State Function DEPRECATED(
    CharacterPartAnimInstance=Self-Reference()
    GameTimeInSeconds=GetGameTimeInSeconds()
    GravityOverrideParams=Get Gravity Override Params()
    DebugJointName=Pelvis
   )
   > Set Gravity Override(Gravity Override=@n2)
   n3 = Get bIsPlayingEmote()
   > Set FrontBOOL(FrontBOOL:bool=AND Boolean(A=Get bIsInFrontEnd(), B=NOT Boolean(A=@n3))) -> (Output_Get:bool)
   n4 = Get CurrentLOD()
   Set LOD1(execute=@n1.then_1, LOD1:bool="integer >= integer"(A=@n4, B=1)) -> (Output_Get:bool)
   Set LOD2(execute=@n1.then_2, LOD2:bool="integer >= integer"(A=@n4, B=2)) -> (Output_Get:bool)
   Set RBAN_Disabled(execute=@n1.then_3, RBAN_Disabled:bool=NOT Boolean(A=Get bRigidBodyEnabled())) -> (Output_Get:bool)
   n5 = Get bIsOnGrindRail()
   Set On Grind Rail(execute=@n1.then_4, On Grind Rail:bool=@n5) -> (Output_Get:bool)
   Set Rban Disabled on Grind Rail(execute=@n1.then_5, Rban Disabled on Grind Rail:bool=NOT Boolean(A=@n5)) -> (Output_Get:bool)
   n6 = CheckForBackbling()
   Set Wear Backpack(execute=@n1.then_6, Wear Backpack:bool=AND Boolean(A=@n6, B=NOT Boolean(A=@n6.bIsBackblingLong))) -> (Output_Get:bool)
  ]}
]
variables [
 FrontBOOL {type=boolean}
 Gravity Override {type=Vector}
 Gravity Override Params {
  type=GravityOverrideParamsStruct
  default {
   dBNO {jointName=root, gravityOverride=<0 0 -980>}
   skydive_FloatingDown {gravityOverride=<0 -700 100>, windAmplitude=<200 -300 200>, windOffset=<0 -600 300>}
   skydive_GoingUp {gravityOverride=<0 -600 -900>, windAmplitude=<200 -200 200>, windOffset=<0 -600 300>}
   skydive_Parachuting {gravityOverride=<0 -500 -980>, windAmplitude=<200 -200 200>}
  }
 }
 LOD1 {type=boolean}
 LOD2 {type=boolean}
 On Grind Rail {type=boolean}
 Rban Disabled on Grind Rail {type=boolean}
 RBAN_Disabled {type=boolean}
 Wear Backpack {type=boolean}
]
