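# Physics asset for the MetroPunk medium female body, providing capsule colliders and constraints for the dynamic jacket "cape" bones.
# Pelvis, lower spine and leg capsules form the body colliders, joined by free-swinging constraints.
# Four two-link jacket chains (back and side, left and right) hang with angular limits from zero-size base capsules tied to the pelvis.
# Cross constraints link each back tip to its side tip, and each side tip to its thigh.
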PhysicsAsset /BRCosmetics/Characters/Player/Female/Medium/Bodies/F_MED_MetroPunk/Meshes/F_MED_MetroPunk_CapeColiders
bodies {
 calf_l [{shapeType=Capsule, length=39.1339, radius=6.1, shapeName=calf_l_capsule, center=<-26.3646 -2.7495 0>, rotation=<rot 90 0 180>}]
 calf_r [{shapeType=Capsule, length=39.1339, radius=6.1, shapeName=calf_r_capsule, center=<26.3646 2.7495 0>, rotation=<rot 90 0 0>}]
 dyn_jacket_back_01_l [{
   shapeType=Capsule
   length=13.4
   radius=6.3
   shapeName=dyn_jacket_side_02_l_capsule
   center=<-6.3074 0.2221 2.0701>
   rotation=<rot 0.0001 179.9998 86.6815>
  }]
 dyn_jacket_back_01_r [{shapeType=Capsule, length=13.4, radius=6.3, shapeName=dyn_jacket_back_01_r_capsule, center=<6.5264 0 -1.7642>, rotation=<rot 0 0 -90>}]
 dyn_jacket_back_02_l [{
   shapeType=Capsule
   length=12.8
   radius=6.9
   shapeName=dyn_jacket_side_02_l_capsule
   center=<-6.3074 0.3483 4.2458>
   rotation=<rot 0.0399 -179.3148 86.6817>
  }]
 dyn_jacket_back_02_r [{shapeType=Capsule, length=12.8, radius=6.9, shapeName=dyn_jacket_back_02_r_capsule, center=<6.5264 -0.583 -3.9436>, rotation=<rot 0 0 -90>}]
 dyn_jacket_back_base_l [{shapeType=Capsule, length=0, radius=0, shapeName=dyn_jacket_back_base_l_capsule, center=<-4.8333 0 0>, rotation=<rot 0 -90 0>}]
 dyn_jacket_back_base_r [{shapeType=Capsule, length=0, radius=0, shapeName=dyn_jacket_back_base_r_capsule, center=<-4.8333 0 0>, rotation=<rot 0 -90 0>}]
 dyn_jacket_side_01_l [{
   shapeType=Capsule
   length=17.9
   radius=6.9
   shapeName=dyn_jacket_side_02_l_capsule
   center=<-6.2815 0.1659 2.2562>
   rotation=<rot 0 179.9999 88.0329>
  }]
 dyn_jacket_side_01_r [{shapeType=Capsule, length=17.9, radius=6.9, shapeName=dyn_jacket_side_01_r_capsule, center=<5.9619 0 -2.2247>, rotation=<rot 0 0 -90>}]
 dyn_jacket_side_02_l [{
   shapeType=Capsule
   length=17.9
   radius=6.9
   shapeName=dyn_jacket_side_02_l_capsule
   center=<-6.2815 0.1659 2.2562>
   rotation=<rot 0 179.9999 88.0329>
  }]
 dyn_jacket_side_02_r [{shapeType=Capsule, length=17.9, radius=6.9, shapeName=dyn_jacket_side_02_r_capsule, center=<5.9619 0 -2.4966>, rotation=<rot 0 0 -90>}]
 dyn_jacket_side_base_l [{shapeType=Capsule, length=0, radius=0, shapeName=dyn_jacket_side_base_l_capsule, center=<-4.8333 0 0>, rotation=<rot 0 -90 0>}]
 dyn_jacket_side_base_r [{shapeType=Capsule, length=0, radius=0, shapeName=dyn_jacket_side_base_r_capsule, center=<-4.8333 0 0>, rotation=<rot 0 -90 0>}]
 foot_l [{shapeType=Capsule, length=15.6124, radius=6.057, shapeName=foot_l_capsule, center=<-4.1441 7.3777 0.1059>, rotation=<rot 0 2.479 97.605>}]
 foot_r [{shapeType=Capsule, length=15.6124, radius=6.057, shapeName=foot_r_capsule, center=<4.1441 -7.3777 -0.1059>, rotation=<rot 0 2.479 -82.395>}]
 pelvis [{shapeType=Capsule, length=2.9, radius=20.1, shapeName=pelvis_capsule, center=<-3.8018 -0.0254 -0.0104>, rotation=<rot 0.2979 179.9904 -180>}]
 spine_01 [{shapeType=Capsule, length=7.3, radius=17.1, shapeName=spine_01_capsule, center=<6.4396 0 0>}]
 spine_02 [{shapeType=Capsule, length=8.4, radius=15.9, shapeName=spine_01_capsule, center=<6.4155 0 0>}]
 thigh_l [{shapeType=Capsule, length=23.3, radius=8.9, shapeName=thigh_l_capsule, center=<-26.198 0 0.2146>, rotation=<rot 90 0 180>}]
 thigh_r [{shapeType=Capsule, length=23.3, radius=8.9, shapeName=thigh_r_capsule, center=<26.198 0 -0.2146>, rotation=<rot 90 0 0>}]
}
constraints [
 {bone1Name=dyn_jacket_back_01_l, bone2Name=dyn_jacket_back_base_l, swing1LimitDegrees=15, swing2LimitDegrees=90, twistLimitDegrees=5}
 {bone1Name=dyn_jacket_back_02_l, bone2Name=dyn_jacket_back_01_l, swing1LimitDegrees=15, swing2LimitDegrees=45, twistLimitDegrees=5}
 {bone1Name=dyn_jacket_back_01_r, bone2Name=dyn_jacket_back_base_r, swing1LimitDegrees=15, swing2LimitDegrees=90, twistLimitDegrees=5}
 {bone1Name=dyn_jacket_back_02_r, bone2Name=dyn_jacket_back_01_r, swing1LimitDegrees=15, swing2LimitDegrees=45, twistLimitDegrees=5}
 {bone1Name=dyn_jacket_side_01_l, bone2Name=dyn_jacket_side_base_l, swing1LimitDegrees=35, swing2LimitDegrees=90, twistLimitDegrees=5}
 {bone1Name=dyn_jacket_side_02_l, bone2Name=dyn_jacket_side_01_l, swing1LimitDegrees=15, swing2LimitDegrees=90, twistLimitDegrees=5}
 {bone1Name=dyn_jacket_side_01_r, bone2Name=dyn_jacket_side_base_r, swing1LimitDegrees=35, swing2LimitDegrees=90, twistLimitDegrees=5}
 {bone1Name=dyn_jacket_side_02_r, bone2Name=dyn_jacket_side_01_r, swing1LimitDegrees=15, swing2LimitDegrees=90, twistLimitDegrees=5}
 {bone1Name=dyn_jacket_back_02_l, bone2Name=dyn_jacket_side_02_l, swing1LimitDegrees=30, swing2LimitDegrees=45, twistLimitDegrees=45}
 {bone1Name=dyn_jacket_back_02_r, bone2Name=dyn_jacket_side_02_r, swing1LimitDegrees=30, swing2LimitDegrees=45, twistLimitDegrees=45}
 {
  bone1Name=spine_01
  bone2Name=pelvis
  swing1LimitDegrees=45
  swing1Motion=Free
  swing2LimitDegrees=45
  swing2Motion=Free
  twistLimitDegrees=45
  twistMotion=Free
 }
 {
  bone1Name=thigh_l
  bone2Name=pelvis
  swing1LimitDegrees=45
  swing1Motion=Free
  swing2LimitDegrees=45
  swing2Motion=Free
  twistLimitDegrees=45
  twistMotion=Free
 }
 {
  bone1Name=thigh_r
  bone2Name=pelvis
  swing1LimitDegrees=45
  swing1Motion=Free
  swing2LimitDegrees=45
  swing2Motion=Free
  twistLimitDegrees=45
  twistMotion=Free
 }
 {
  bone1Name=dyn_jacket_side_base_r
  bone2Name=pelvis
  swing1LimitDegrees=45
  swing1Motion=Free
  swing2LimitDegrees=45
  swing2Motion=Free
  twistLimitDegrees=45
  twistMotion=Free
 }
 {
  bone1Name=dyn_jacket_back_base_r
  bone2Name=pelvis
  swing1LimitDegrees=45
  swing1Motion=Free
  swing2LimitDegrees=45
  swing2Motion=Free
  twistLimitDegrees=45
  twistMotion=Free
 }
 {
  bone1Name=dyn_jacket_side_base_l
  bone2Name=pelvis
  swing1LimitDegrees=45
  swing1Motion=Free
  swing2LimitDegrees=45
  swing2Motion=Free
  twistLimitDegrees=45
  twistMotion=Free
 }
 {
  bone1Name=dyn_jacket_back_base_l
  bone2Name=pelvis
  swing1LimitDegrees=45
  swing1Motion=Free
  swing2LimitDegrees=45
  swing2Motion=Free
  twistLimitDegrees=45
  twistMotion=Free
 }
 {bone1Name=thigh_l, bone2Name=dyn_jacket_side_02_l, swing1LimitDegrees=25, swing2LimitDegrees=35, twistLimitDegrees=35}
 {bone1Name=thigh_r, bone2Name=dyn_jacket_side_02_r, swing1LimitDegrees=15, swing2LimitDegrees=15, twistLimitDegrees=15}
 {
  bone1Name=thigh_l
  bone2Name=thigh_r
  swing1LimitDegrees=45
  swing1Motion=Free
  swing2LimitDegrees=45
  swing2Motion=Free
  twistLimitDegrees=45
  twistMotion=Free
 }
 {
  bone1Name=calf_r
  bone2Name=thigh_r
  swing1LimitDegrees=45
  swing1Motion=Free
  swing2LimitDegrees=45
  swing2Motion=Free
  twistLimitDegrees=45
  twistMotion=Free
 }
 {
  bone1Name=calf_l
  bone2Name=thigh_l
  swing1LimitDegrees=45
  swing1Motion=Free
  swing2LimitDegrees=45
  swing2Motion=Free
  twistLimitDegrees=45
  twistMotion=Free
 }
 {
  bone1Name=spine_02
  bone2Name=spine_01
  swing1LimitDegrees=45
  swing1Motion=Free
  swing2LimitDegrees=45
  swing2Motion=Free
  twistLimitDegrees=45
  twistMotion=Free
 }
 {
  bone1Name=foot_r
  bone2Name=calf_r
  swing1LimitDegrees=45
  swing1Motion=Free
  swing2LimitDegrees=45
  swing2Motion=Free
  twistLimitDegrees=45
  twistMotion=Free
 }
 {
  bone1Name=foot_l
  bone2Name=calf_l
  swing1LimitDegrees=45
  swing1Motion=Free
  swing2LimitDegrees=45
  swing2Motion=Free
  twistLimitDegrees=45
  twistMotion=Free
 }
]
