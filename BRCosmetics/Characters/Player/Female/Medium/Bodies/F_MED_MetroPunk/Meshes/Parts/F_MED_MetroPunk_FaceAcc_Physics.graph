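# Physics asset for the face accessory part of the female medium MetroPunk player body, covering two pairs of dynamic side hair strands around the head.
# Each strand has a zero-size base body and a capsule body joined by tight swing and twist limits, alongside a head capsule.
# Base bodies connect to the head with free swing and twist, except the first left base, which has no head constraint listed.
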PhysicsAsset /BRCosmetics/Characters/Player/Female/Medium/Bodies/F_MED_MetroPunk/Meshes/Parts/F_MED_MetroPunk_FaceAcc_Physics
bodies {
 dyn_hair_side_01_base_l [{shapeType=Capsule, length=0, radius=0, shapeName=dyn_hair_side_01_base_l_capsule}]
 dyn_hair_side_01_base_r [{shapeType=Capsule, length=0, radius=0, shapeName=dyn_hair_side_01_base_r_capsule}]
 dyn_hair_side_01_l [{
   shapeType=Capsule
   length=3.1
   radius=3.1
   shapeName=dyn_hair_side_01_base_l_capsule
   center=<0.6004 0 0>
   rotation=<rot -1.3064 5.4914 -76.6578>
  }]
 dyn_hair_side_01_r [{
   shapeType=Capsule
   length=4
   radius=2.2
   shapeName=dyn_hair_side_01_base_l_capsule
   center=<0.9759 0.1486 0.2085>
   rotation=<rot -13.3998 88.3866 -85.639>
  }]
 dyn_hair_side_02_base_l [{shapeType=Capsule, length=0, radius=0, shapeName=dyn_hair_side_02_base_l_capsule}]
 dyn_hair_side_02_base_r [{shapeType=Capsule, length=0, radius=0, shapeName=dyn_hair_side_02_base_r_capsule}]
 dyn_hair_side_02_l [{
   shapeType=Capsule
   length=3.8
   radius=2.4
   shapeName=dyn_hair_side_01_base_l_capsule
   center=<-0.171 0.0239 1.2544>
   rotation=<rot 5.9563 -2.3411 -2.2445>
  }]
 dyn_hair_side_02_r [{
   shapeType=Capsule
   length=3.7
   radius=2.5
   shapeName=dyn_hair_side_01_base_l_capsule
   center=<1.6072 0.0001 0.0031>
   rotation=<rot 0 0 -87.7554>
  }]
 head [{shapeType=Capsule, length=5.1536, radius=6.2731, shapeName=head_capsule, rotation=<rot -90 -0.0003 0>}]
}
constraints [
 {bone1Name=dyn_hair_side_01_l, bone2Name=dyn_hair_side_01_base_l, swing1LimitDegrees=5, swing2LimitDegrees=7, twistLimitDegrees=7}
 {
  bone1Name=dyn_hair_side_01_base_r
  bone2Name=head
  swing1LimitDegrees=45
  swing1Motion=Free
  swing2LimitDegrees=45
  swing2Motion=Free
  twistLimitDegrees=45
  twistMotion=Free
 }
 {bone1Name=dyn_hair_side_01_r, bone2Name=dyn_hair_side_01_base_r, swing1LimitDegrees=15, swing2LimitDegrees=15, twistLimitDegrees=5}
 {
  bone1Name=dyn_hair_side_02_base_l
  bone2Name=head
  swing1LimitDegrees=45
  swing1Motion=Free
  swing2LimitDegrees=45
  swing2Motion=Free
  twistLimitDegrees=45
  twistMotion=Free
 }
 {bone1Name=dyn_hair_side_02_l, bone2Name=dyn_hair_side_02_base_l, swing1LimitDegrees=10, swing2LimitDegrees=10, twistLimitDegrees=10}
 {
  bone1Name=dyn_hair_side_02_base_r
  bone2Name=head
  swing1LimitDegrees=45
  swing1Motion=Free
  swing2LimitDegrees=45
  swing2Motion=Free
  twistLimitDegrees=45
  twistMotion=Free
 }
 {bone1Name=dyn_hair_side_02_r, bone2Name=dyn_hair_side_02_base_r, swing1LimitDegrees=5, swing2LimitDegrees=5, twistLimitDegrees=2}
]
